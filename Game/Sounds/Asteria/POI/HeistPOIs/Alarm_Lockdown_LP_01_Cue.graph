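# Looping lockdown alarm sound cue for the Asteria Heist POIs.
# A short randomized delay feeds a looping wave player, offsetting the start of each instance.
# It uses the base alarm attenuation and the gameplay events sound class, and carries concurrency override values that cap simultaneous instances and allow volume-scale release.
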SoundCue /Game/Sounds/Asteria/POI/HeistPOIs/Alarm_Lockdown_LP_01_Cue
root=Delay_0
nodes [
 {id=Delay_0, children [WavePlayer_1], properties {delayMax=0.5, delayMin=0.2}}
 {id=WavePlayer_1, properties {bLooping=true, soundWaveAssetPtr=/Game/Sounds/Asteria/POI/HeistPOIs/wav/Alarm_Lockdown_LP_01}}
]
properties {
 attenuationSettings=&/Att_Base_Alarms
 maxDistance=2900
 soundClassObject=&Fort_Events_Gameplay
 volumeMultiplier=1
 concurrencyOverrides {bVolumeScaleCanRelease=true, maxCount=3}
 concurrencySet [null]
}
